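# Physics asset for the M_MED_Sparrow backpack cosmetic, giving secondary motion to a pack with head, ears, nose, limbs, tail, pendant and stuffing bones.
# Single-capsule bodies form a constraint tree from root through pack to dyn_pack, which carries the head, limbs, tail and pendant.
# The head carries the ears, nose and stuffing.
# Angular limits are tight on the head, looser on tail and stuffing, and mostly single-axis on ears and nose.
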PhysicsAsset /BRCosmetics/Accessories/FORT_Backpacks/Backpack_M_MED_Sparrow/Meshes/M_MED_Sparrow_Pack_Head_Physics
bodies {
 dyn_arm_l1 [{shapeType=Capsule, length=5, radius=2, shapeName=dyn_leg_l1_capsule, center=<6.1063 0 1.4152>, rotation=<rot 75.977 180 180>}]
 dyn_arm_r1 [{shapeType=Capsule, length=2, radius=1, shapeName=dyn_leg_r1_capsule, center=<2.3939 0 0>, rotation=<rot 90 0 0>}]
 dyn_ear_l1 [{shapeType=Capsule, length=2, radius=2, shapeName=dyn_ear_l1_capsule, center=<2 0 0>, rotation=<rot 90 0 0>}]
 dyn_ear_r1 [{shapeType=Capsule, length=2, radius=2, shapeName=dyn_ear_r1_capsule, center=<2 0 0>, rotation=<rot 90 0 0>}]
 dyn_head1 [{shapeType=Capsule, length=2, radius=2, shapeName=dyn_ear_l1_capsule, center=<2 0 0>, rotation=<rot 90 0 0>}]
 dyn_leg_l1 [{shapeType=Capsule, length=4.8264, radius=2.1345, shapeName=dyn_leg_l1_capsule, center=<5.3328 0 0>, rotation=<rot 90 0 0>}]
 dyn_leg_r1 [{shapeType=Capsule, length=4, radius=2.762, shapeName=dyn_leg_r1_capsule, center=<4.265 0 0>, rotation=<rot 90 0 0>}]
 dyn_nose1 [{shapeType=Capsule, length=2, radius=2, shapeName=dyn_nose1_capsule, center=<3 0 0>, rotation=<rot 90 0 0>}]
 dyn_pack [{shapeType=Capsule, length=5, radius=8.8, shapeName=dyn_pack_capsule, center=<11.3109 0 -10.4207>, rotation=<rot -88.3886 -180 180>}]
 dyn_pendant1 [{
   shapeType=Capsule
   length=4.0196
   radius=2.1
   shapeName=dyn_pendant1_capsule
   center=<2.5564 0 -1.8>
   rotation=<rot 88.9998 -179.9994 -179.9994>
  }]
 dyn_stuffing1 [{shapeType=Capsule, length=2, radius=1.9934, shapeName=dyn_stuffing1_capsule, center=<2.9806 0 0>, rotation=<rot 90 0 0>}]
 dyn_tail1 [{shapeType=Capsule, length=2.8, radius=3.3, shapeName=dyn_tail1_capsule, center=<5 0 -6>, rotation=<rot -87 0 0>}]
 pack [{shapeType=Capsule, length=5, radius=5, shapeName=pack_capsule, center=<0 0 -20>}]
 root [{shapeType=Capsule, length=2, radius=2, shapeName=root_capsule}]
}
constraints [
 {bone1Name=pack, bone2Name=root, swing1LimitDegrees=10, swing2LimitDegrees=10, twistLimitDegrees=10}
 {bone1Name=dyn_pack, bone2Name=pack, swing1LimitDegrees=2, swing2LimitDegrees=15, twistLimitDegrees=2}
 {bone1Name=dyn_pendant1, bone2Name=dyn_pack, swing1LimitDegrees=10, swing2LimitDegrees=20, twistLimitDegrees=10}
 {bone1Name=dyn_tail1, bone2Name=dyn_pack, swing1LimitDegrees=45, swing2LimitDegrees=45, twistLimitDegrees=10}
 {bone1Name=dyn_head1, bone2Name=dyn_pack, swing1LimitDegrees=2.5, swing2LimitDegrees=5, twistLimitDegrees=2.5}
 {bone1Name=dyn_ear_l1, bone2Name=dyn_head1, swing1LimitDegrees=0, swing2LimitDegrees=30, twistLimitDegrees=5}
 {bone1Name=dyn_ear_r1, bone2Name=dyn_head1, swing1LimitDegrees=0, swing2LimitDegrees=30, twistLimitDegrees=5}
 {bone1Name=dyn_nose1, bone2Name=dyn_head1, swing1LimitDegrees=0, swing2LimitDegrees=10, twistLimitDegrees=0}
 {bone1Name=dyn_stuffing1, bone2Name=dyn_head1, swing1LimitDegrees=30, swing2LimitDegrees=30, twistLimitDegrees=30}
 {bone1Name=dyn_leg_l1, bone2Name=dyn_pack, swing1LimitDegrees=5, swing2LimitDegrees=15, twistLimitDegrees=5}
 {bone1Name=dyn_leg_r1, bone2Name=dyn_pack, swing1LimitDegrees=5, swing2LimitDegrees=15, twistLimitDegrees=5}
 {bone1Name=dyn_arm_l1, bone2Name=dyn_pack, swing1LimitDegrees=5, swing2LimitDegrees=30, twistLimitDegrees=5}
 {bone1Name=dyn_arm_r1, bone2Name=dyn_pack, swing1LimitDegrees=5, swing2LimitDegrees=35, twistLimitDegrees=5}
]
